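# Translucent, unlit, two-sided material for Niagara mesh particles, by its name the slash effect in the Keytar Ninja emote.
# Color is a rotated two-color gradient, desaturated by a dynamic parameter, tinted by particle color and passed through MF_FX_EmissiveAdjustments.
# A procedural alpha mask intersects two hard-edged, oppositely rotated UV bands, trims them with step cuts and multiplies by particle alpha.
# The Size-driven mask and inverted Fresnel feed nothing shown.
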Material /SparksCosmetics/Emotes/Keytar/Ninja/Effects/M_Emote_Keytar_Ninja_Slash
graph {
 properties {
  bAutomaticallySetUsageInEditor=false
  bUsedWithNiagaraMeshParticles=true
  blendMode=BLEND_Translucent
  shadingModel=MSM_Unlit
  twoSided=true
  usedShadingModels=Unlit
 }
 nodes [
  n1 = MaterialFunctionCall(UVs=TextureCoordinate(), "Rotation Angle (0-1)"=ScalarParameter() {defaultValue=0.25, parameterName=Rotate Gradient}) {materialFunction=&CustomRotator}
  n2 = LinearInterpolate(
   A=Multiply(
    A=VectorParameter() {parameterName=Color 1, defaultValue=<rgba 1 0 0 1>}.RGB
    B=ScalarParameter() {defaultValue=1, parameterName=Color 1 Strength}
   )
   B=Multiply(
    A=VectorParameter() {parameterName=Color 2, defaultValue=<rgba 0 0 1 1>}.RGB
    B=ScalarParameter() {defaultValue=1, parameterName=Color 2 Strength}
   )
   Alpha=SmoothStep(
    Min=ScalarParameter() {parameterName=SmoothStep Min}
    Max=ScalarParameter() {defaultValue=1, parameterName=SmoothStep Max}
    Value=StaticComponentMaskParameter(None=@n1.Rotated Values) {defaultR=true, parameterName=Gradient Channel Mask}
   )
  )
  n3 = DynamicParameter() {defaultValue=<rgba 0 1 1 1>, paramNames [Desat,Param2,Param3,Param4]}
  n4 = ParticleColor()
  MaterialFunctionCall(Emissive=Multiply(A=Desaturation(None=@n2, Fraction=@n3.Desat), B=@n4.RGB)) {materialFunction=&MF_FX_EmissiveAdjustments}
  n5 = TextureCoordinate() {vTiling=0.5}
  n6 = Constant() {r=0.23}
  n7 = Add(A=Multiply(A=MaterialFunctionCall(UVs=@n5, "Rotation Angle (0-1)"=@n6) {materialFunction=&CustomRotator}.Rotated Values) {constB=2}) {constB=-1}
  n8 = MaterialFunctionCall(Float2=@n7) {materialFunction=&BreakOutFloat2Components}
  MaterialFunctionCall(Vector 2=@n8.G) {materialFunction=/Engine/Functions/Engine_MaterialFunctions02/Utility/VectorLength}
  n9 = ScalarParameter() {defaultValue=0.7, parameterName=Size}
  n10 = MaterialFunctionCall(UVs=TextureCoordinate() {vTiling=0.5}, "Rotation Angle (0-1)"=Constant() {r=0.875}) {materialFunction=&CustomRotator}
  n11 = MaterialFunctionCall(Float2=Add(A=Multiply(A=@n10.Rotated Values) {constB=2}) {constB=-1}) {materialFunction=&BreakOutFloat2Components}
  n12 = MaterialFunctionCall(Vector 2=@n11.R) {materialFunction=&VectorLength}
  n13 = MaterialFunctionCall(Vector 2=@n11.G) {materialFunction=&VectorLength}
  Multiply(A=OneMinus(None=SmoothStep(Min=@n9, Max=@n9, Value=@n12.V2 Length)), B=OneMinus(None=SmoothStep(Min=@n9, Max=@n9, Value=@n13.V2 Length)))
  n14 = Add(A=Step(X=@n11.R), B=Step(X=@n11.G))
  n15 = Constant() {r=0.95}
  n16 = MaterialFunctionCall(Vector 2=@n8.R) {materialFunction=&VectorLength}
  n17 = Multiply(A=MaterialFunctionCall(UVs=TextureCoordinate() {vTiling=0.5}, "Rotation Angle (0-1)"=Multiply(A=@n6) {constB=-1}) {materialFunction=&CustomRotator}.Rotated Values) {constB=2}
  n18 = MaterialFunctionCall(Vector 2=MaterialFunctionCall(Float2=Add(A=@n17) {constB=-1}) {materialFunction=&BreakOutFloat2Components}.R) {materialFunction=&VectorLength}
  n19 = Multiply(A=OneMinus(None=SmoothStep(Min=@n15, Max=@n15, Value=@n16.V2 Length)), B=OneMinus(None=SmoothStep(Min=@n15, Max=@n15, Value=@n18.V2 Length)))
  Multiply(A=@n4.A, B=Multiply(A=OneMinus(None=@n14), B=Saturate(None=@n19)))
  OneMinus(None=Fresnel(
    ExponentIn=ScalarParameter() {defaultValue=5, parameterName=Fresnel ExponentIn}
    BaseReflectFractionIn=ScalarParameter() {defaultValue=0.04, parameterName=BaseReflectFractionIn}
   ) {exponent=0})
  OneMinus(None=@n7)
 ]
}
