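# Niagara system for a brief sprite glint in the Ninja Keytar emote from Sparks cosmetics.
# A single Glint emitter burst-spawns short-lived sprites, with size and camera offset exposed as User parameters.
# Each sprite pops up fast on a scale curve, shrinks away while spinning at a fixed rate, and has its alpha dimmed by time of day.
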NiagaraSystem /SparksCosmetics/Emotes/Keytar/Ninja/Effects/NS_Emote_Keytar_Ninja_Glint
dependencies {usedDataInterfaces [&NiagaraDataInterfaceCurve], usedDynamicInputs [], usedModules [&SystemState]}
emitters {Glint {
  emitterUpdateScript [
   &EmitterState()
   &SpawnBurst_Instantaneous(Use Spawn Probability=false, UseLoopCountLimit=true)
  ]
  particleSpawnScript [
   &InitializeParticle(
    Color=<rgba 0.8958 0.8958 0.8958 1>
    Lifetime=0.4
    Sprite Rotation Mode=::NewEnumerator2
    Sprite Size Mode=::NewEnumerator3
    Uniform Sprite Size=$User.SparkleSize
    UsePositionOffset=false
    Write Material Random=true
   )
   &CameraOffset(Camera Offset Amount=$User.SparkleCameraOffset)
  ]
  particleUpdateScript [
   &ParticleState()
   &ScaleSpriteSize(
    Uniform Curve Scale=1.2
    Uniform Curve Sprite Scale=DataInterface() {propertyValues {Curve {keys [
        {arriveTangent=9.0725, interpMode=RCIM_Cubic, leaveTangent=9.0725, tangentMode=RCTM_User, time=0, value=0}
        {arriveTangent=-1.5354, interpMode=RCIM_Cubic, leaveTangent=-1.5354, tangentMode=RCTM_User, time=0.0666, value=0.9847}
        {arriveTangent=-0.0405, interpMode=RCIM_Cubic, leaveTangent=-0.0405, tangentMode=RCTM_User, time=1, value=0}
       ]}}}
   )
   &SpriteRotationRate(Rotation Rate=&RandomRangeFloat(Minimum=360, Maximum=360))
   &ScaleColor(
    Scale Alpha=&Multiply_Float(A=1, B=&FNTimeOfDayScalar(Night Scale=0.25, Morning Scale=0.4, Evening Scale=0.4))
    ScaleA=true
    ScaleRGB=false
   )
  ]
  rendererClasses [&NiagaraSpriteRendererProperties]
  renderers [{rendererClass=&NiagaraSpriteRendererProperties}]
 }}
summary {
 systemName=NS_Emote_Keytar_Ninja_Glint
 emitters [{emitterName=Glint, rendererClasses [&NiagaraSpriteRendererProperties]}]
 userVariables [
  User.SparkleCameraOffset {type=NiagaraFloat, defaultValue=40, description=""}
  User.SparkleSize {type=NiagaraFloat, defaultValue=60, description=""}
 ]
}
