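# Niagara slash effect for the Ninja Keytar emote in Sparks cosmetics, built from three mesh-rendered burst emitters.
# Slash_01 and a thinner, offset Slash_02 on a loop delay pop in with curve-driven scale, then fade while shifting from white to red.
# Action_Lines adds two brief stretched streaks.
# Slash brightness follows time-of-day scalars with a front-end preview input, and Slash_01 also animates a dynamic material parameter.
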NiagaraSystem /SparksCosmetics/Emotes/Keytar/Ninja/Effects/NS_Emote_Keytar_Ninja_Slash
dependencies {
 usedDataInterfaces [
  &NiagaraDataInterfaceCurve
  &NiagaraDataInterfaceColorCurve
  &NiagaraDataInterfaceArrayFloat
  &NiagaraDataInterfaceVector2DCurve
 ]
 usedDynamicInputs [&MakeFloatFromVector2d]
 usedModules [&SystemState]
}
emitters {
 Action_Lines {
  emitterUpdateScript [
   &EmitterState()
   &SpawnBurst_Instantaneous(Spawn Count=2, Use Spawn Probability=false, UseLoopCountLimit=false)
  ]
  particleSpawnScript [
   &InitializeParticle(
    Color=<rgba 0 0 0 20>
    Lifetime Max=0.2
    Lifetime Min=0.1
    Lifetime Mode=::NewEnumerator1
    Mesh Scale=&MakeVector(
     X=&RandomRangeFloat(Minimum=1, Maximum=1)
     Y=&RandomRangeFloat(Minimum=1, Maximum=1)
     Z=&RandomRangeFloat(Minimum=2.25, Maximum=2.25)
    )
    Mesh Scale Mode=::NewEnumerator4
    Sprite Rotation Angle=&SelectFloatFromArray(
     Array Sampling Mode=::NewEnumerator0
     Direct Array Index=?
     Float Selection Array=DataInterface() {propertyValues {FloatData [90,270]}}
     Random Seed=0
    )
    Sprite Rotation Mode=::NewEnumerator0
    Sprite Size=&Multiply_Vector2D_ByFloat(
     Float=&FN_FrontEndFloatModifier(Gameplay=1.5, Front End=1.2)
     Vector2D=&RandomRangeVector2D(Minimum=<40 200>, Maximum=<60 250>)
    )
    Sprite Size Mode=::NewEnumerator4
    UsePositionOffset=false
    Write Material Random=true
   )
   &ShapeLocation(
    Box Size=<10 10 10>
    MaskSpawnGroup=false
    Override Seed=false
    Rotation Mode=::NewEnumerator1
    Shape Primitive=::NewEnumerator2
    Surface Only=false
    UseOwnerScale=true
   ) {module=ShapeLocation001}
   &InitialMeshOrientation(Mesh Orientation Mode=::NewEnumerator2, Orientation Vector=<0 0 0.25>, Use Rotation Vector=true)
   &CameraOffset(Camera Offset Amount=30)
   &SpriteFacingAndAlignment(Sprite Alignment=<0 0 1>, WriteAlignment=true, WriteFacing=false)
  ]
  particleUpdateScript [
   &ParticleState()
   &ScaleMeshSize(Scale Factor=&MakeVector(
     X=&FloatFromCurve(FloatCurve=DataInterface() {propertyValues {Curve {keys [<at 0 0.5>,<at 1 1.5>]}}})
     Y=&FloatFromCurve(FloatCurve=DataInterface() {propertyValues {Curve {keys [<at 0 0.5>,<at 1 1.5>]}}})
     Z=&FloatFromCurve(FloatCurve=DataInterface() {propertyValues {Curve {keys [<at 0 1>,<at 1 0.5>]}}})
    )) {module=ScaleMeshSize001}
   &SolveForcesAndVelocity(Clamp Velocity=false, Limit Acceleration=false)
   &ScaleColor(
    Scale Alpha=&FloatFromCurve(FloatCurve=DataInterface() {propertyValues {Curve {keys [
         {interpMode=RCIM_Cubic, tangentMode=RCTM_User, time=0, value=0}
         {interpMode=RCIM_Cubic, tangentMode=RCTM_User, time=0.5, value=1}
         {tangentMode=RCTM_User, time=1, value=0}
        ]}}})
    ScaleA=true
    ScaleRGB=true
   )
   &ScaleSpriteSize(
    Non-Uniform Curve Sprite Scale=DataInterface() {propertyValues {
      XCurve {keys [{tangentMode=RCTM_User, time=0, value=0.5}, {tangentMode=RCTM_User, time=1, value=1.5}]}
      YCurve {keys [{tangentMode=RCTM_User, time=0, value=1}, {tangentMode=RCTM_User, time=1, value=0.5}]}
     }}
    Scale Sprite Size Mode=::NewEnumerator3
   )
  ]
  rendererClasses [&NiagaraMeshRendererProperties]
  renderers [{rendererClass=&NiagaraMeshRendererProperties, rendererIndex=0}]
 }
 Slash_01 {
  emitterUpdateScript [
   &EmitterState()
   &SpawnBurst_Instantaneous(Use Spawn Probability=false, UseLoopCountLimit=true)
  ]
  particleSpawnScript [
   &InitializeParticle(
    Lifetime=0.5
    Mesh Scale=<0.065 0.065 0.8>
    Mesh Scale Mode=::NewEnumerator4
    Sprite Rotation Mode=::NewEnumerator0
    Sprite Size=&FN_FrontEndVector2Modifier(Front End:Vector2f=<110 1.2>, Gameplay:Vector2f=<150 2>)
    Sprite Size Mode=::NewEnumerator4
    UsePositionOffset=true
    Write Material Random=true
   )
   &InitialMeshOrientation(Mesh Orientation Mode=::NewEnumerator2, Orientation Vector=<0 0 0.25>, Use Rotation Vector=true)
   &CameraOffset(Camera Offset Amount=50)
  ]
  particleUpdateScript [
   &ParticleState()
   &ScaleMeshSize(Scale Factor=&MakeVector(
     X=&FloatFromCurve(FloatCurve=DataInterface() {propertyValues {Curve {keys [
          {arriveTangent=21.0661, interpMode=RCIM_Cubic, leaveTangent=21.0661, tangentMode=RCTM_User, time=0, value=0.5}
          {arriveTangent=-16.9436, interpMode=RCIM_Cubic, leaveTangent=-16.9436, tangentMode=RCTM_User, time=0.1, value=3}
          {arriveTangent=-4.3657, interpMode=RCIM_Cubic, leaveTangent=-4.3657, tangentMode=RCTM_User, time=0.2, value=1}
          {arriveTangent=-0.1426, interpMode=RCIM_Cubic, leaveTangent=-0.1426, tangentMode=RCTM_User, time=0.9, value=0}
         ]}}})
     Y=&FloatFromCurve(FloatCurve=DataInterface() {propertyValues {Curve {keys [
          {arriveTangent=21.0661, interpMode=RCIM_Cubic, leaveTangent=21.0661, tangentMode=RCTM_User, time=0, value=0.5}
          {arriveTangent=-16.9436, interpMode=RCIM_Cubic, leaveTangent=-16.9436, tangentMode=RCTM_User, time=0.1, value=3}
          {arriveTangent=-4.5167, interpMode=RCIM_Cubic, leaveTangent=-4.5167, tangentMode=RCTM_User, time=0.2, value=1}
          {arriveTangent=-0.1426, interpMode=RCIM_Cubic, leaveTangent=-0.1426, tangentMode=RCTM_User, time=0.9, value=0}
         ]}}})
     Z=&FloatFromCurve(FloatCurve=DataInterface() {propertyValues {Curve {keys [
          {arriveTangent=23.7098, interpMode=RCIM_Cubic, leaveTangent=23.7098, tangentMode=RCTM_User, time=0, value=0}
          {arriveTangent=1.5, interpMode=RCIM_Cubic, leaveTangent=1.5, time=0.1, value=1}
          {interpMode=RCIM_Cubic, leaveTangent=-0.0869, time=1, value=1.5}
         ]}}})
    )) {module=ScaleMeshSize003}
   &ScaleColor(
    Scale Alpha=&FloatFromCurve(FloatCurve=DataInterface() {propertyValues {Curve {keys [
         {arriveTangent=9.0725, interpMode=RCIM_Cubic, leaveTangent=9.0725, tangentMode=RCTM_User, time=0, value=0}
         {arriveTangent=-0.04, interpMode=RCIM_Cubic, leaveTangent=-0.04, tangentMode=RCTM_User, time=0.1, value=1}
         {arriveTangent=-0.9095, interpMode=RCIM_Cubic, leaveTangent=-0.9095, tangentMode=RCTM_User, time=1, value=0}
        ]}}})
    Scale RGB=&Multiply_VectorByFloat(
     Vector=&MakeVectorFromLinearColorRGB(Color=&ColorFromCurve(ColorCurve=DataInterface() {propertyValues {
         AlphaCurve {keys []}
         BlueCurve {keys [<at 0 1>, {interpMode=RCIM_Cubic, leaveTangent=-1.6899, time=0.6197, value=0}]}
         GreenCurve {keys [<at 0 1>, {interpMode=RCIM_Cubic, leaveTangent=-1.6899, time=0.6197, value=0}]}
         RedCurve {keys [<at 0 1>,<at 0.6197 1 RCIM_Cubic>]}
        }}))
     Float=&FNTimeOfDayScalar(
      Night Scale=0.5
      Morning Scale=0.65
      Day Scale=&MakeCustomFloatFromBool(Bool=$User.bIsFrontEndPreview, True Float=0.66, False Float=1)
      Evening Scale=0.65
     )
    )
    ScaleA=true
    ScaleRGB=true
   )
   &SolveForcesAndVelocity(Clamp Velocity=false, Limit Acceleration=false)
   &DynamicMaterialParameters(
    Index 0 Param 1=&FloatFromCurve(FloatCurve=DataInterface() {propertyValues {Curve {keys [<at 0 1 RCIM_Cubic>, {arriveTangent=-13.5622, interpMode=RCIM_Cubic, leaveTangent=-13.5622, tangentMode=RCTM_User, time=0.15, value=0}]}}})
    Module.Index0Param0Write=true
    Module.Index0Param1Write=true
    Module.Index0Param2Write=true
    Module.Index0Param3Write=true
   )
  ]
  rendererClasses [&NiagaraMeshRendererProperties]
  renderers [{rendererClass=&NiagaraMeshRendererProperties, rendererIndex=0}]
 }
 Slash_02 {
  emitterUpdateScript [
   &EmitterState(Life Cycle Mode=::NewEnumerator1, Loop Behavior=::NewEnumerator1, Loop Delay=0.1, Loop Duration Mode=::NewEnumerator1, UseLoopDelay=true)
   &SpawnBurst_Instantaneous(Use Spawn Probability=false, UseLoopCountLimit=true)
  ]
  particleSpawnScript [
   &InitializeParticle(
    Color=<rgba 10 10 10 1>
    Lifetime=0.5
    Mesh Scale=<0.015 0.015 0.7>
    Mesh Scale Mode=::NewEnumerator4
    Position Offset=<-80 0 0>
    Sprite Rotation Mode=::NewEnumerator0
    Sprite Size=&FN_FrontEndVector2Modifier(Front End:Vector2f=<110 1.2>, Gameplay:Vector2f=<150 2>)
    Sprite Size Mode=::NewEnumerator4
    UsePositionOffset=true
    Write Material Random=true
   )
   &InitialMeshOrientation(Mesh Orientation Mode=::NewEnumerator2, Orientation Vector=<0 0 0>, Rotation=<0 0.25 0>, Use Rotation Vector=true)
   &CameraOffset(Camera Offset Amount=50)
  ]
  particleUpdateScript [
   &ParticleState()
   &ScaleSpriteSize(Uniform Curve Sprite Scale=DataInterface() {propertyValues {Curve {keys [
        <at 0 0.5 RCIM_Cubic>
        {arriveTangent=-2.8328, interpMode=RCIM_Cubic, leaveTangent=-2.8328, tangentMode=RCTM_User, time=0.1, value=1.5}
        {interpMode=RCIM_Cubic, leaveTangent=4.5, time=0.3, value=1}
       ]}}})
   &ScaleMeshSize(Scale Factor=&MakeVector(
     X=&FloatFromCurve(FloatCurve=DataInterface() {propertyValues {Curve {keys [
          {arriveTangent=21.0661, interpMode=RCIM_Cubic, leaveTangent=21.0661, tangentMode=RCTM_User, time=0, value=0.5}
          {arriveTangent=-16.9436, interpMode=RCIM_Cubic, leaveTangent=-16.9436, tangentMode=RCTM_User, time=0.1, value=3}
          {arriveTangent=0.0582, interpMode=RCIM_Cubic, leaveTangent=0.0582, tangentMode=RCTM_User, time=0.2, value=1}
          {arriveTangent=-2.3804, interpMode=RCIM_Cubic, leaveTangent=-2.3804, tangentMode=RCTM_User, time=1, value=0}
         ]}}})
     Y=&FloatFromCurve(FloatCurve=DataInterface() {propertyValues {Curve {keys [
          {arriveTangent=21.0661, interpMode=RCIM_Cubic, leaveTangent=21.0661, tangentMode=RCTM_User, time=0, value=0.5}
          {arriveTangent=-16.9436, interpMode=RCIM_Cubic, leaveTangent=-16.9436, tangentMode=RCTM_User, time=0.1, value=3}
          {arriveTangent=0.0582, interpMode=RCIM_Cubic, leaveTangent=0.0582, tangentMode=RCTM_User, time=0.2, value=1}
          {arriveTangent=-2.3804, interpMode=RCIM_Cubic, leaveTangent=-2.3804, tangentMode=RCTM_User, time=1, value=0}
         ]}}})
     Z=&FloatFromCurve(FloatCurve=DataInterface() {propertyValues {Curve {keys [
          {arriveTangent=21.0661, interpMode=RCIM_Cubic, leaveTangent=21.0661, tangentMode=RCTM_User, time=0, value=0.5}
          {arriveTangent=-16.9436, interpMode=RCIM_Cubic, leaveTangent=-16.9436, tangentMode=RCTM_User, time=0.1, value=2}
          {arriveTangent=0.0582, interpMode=RCIM_Cubic, leaveTangent=0.0582, tangentMode=RCTM_User, time=0.175, value=0.7}
         ]}}})
    ))
   &ScaleColor(
    Scale Alpha=&FloatFromCurve(FloatCurve=DataInterface() {propertyValues {Curve {keys [
         {arriveTangent=9.0725, interpMode=RCIM_Cubic, leaveTangent=9.0725, tangentMode=RCTM_User, time=0, value=0}
         {arriveTangent=-0.04, interpMode=RCIM_Cubic, leaveTangent=-0.04, tangentMode=RCTM_User, time=0.1, value=1}
         {arriveTangent=-0.9095, interpMode=RCIM_Cubic, leaveTangent=-0.9095, tangentMode=RCTM_User, time=1, value=0}
        ]}}})
    Scale RGB=&Multiply_VectorByFloat(
     Vector=&MakeVectorFromLinearColorRGB(Color=&ColorFromCurve(ColorCurve=DataInterface() {propertyValues {
         AlphaCurve {keys []}
         BlueCurve {keys [<at 0 1>, {interpMode=RCIM_Cubic, leaveTangent=-1.6899, time=0.6197, value=0}]}
         GreenCurve {keys [<at 0 1>, {interpMode=RCIM_Cubic, leaveTangent=-1.6899, time=0.6197, value=0}]}
         RedCurve {keys [<at 0 1>,<at 0.6197 1 RCIM_Cubic>]}
        }}))
     Float=&FNTimeOfDayScalar(
      Night Scale=0.5
      Morning Scale=0.65
      Day Scale=&MakeCustomFloatFromBool(Bool=$User.bIsFrontEndPreview, True Float=1, False Float=1)
      Evening Scale=0.65
     )
    )
    ScaleA=true
    ScaleRGB=true
   )
   &SolveForcesAndVelocity(Clamp Velocity=false, Limit Acceleration=false)
  ]
  rendererClasses [&NiagaraMeshRendererProperties]
  renderers [{rendererClass=&NiagaraMeshRendererProperties, rendererIndex=0}]
 }
}
summary {
 systemName=NS_Emote_Keytar_Ninja_Slash
 emitters [
  {emitterName=Slash_02, rendererClasses [&NiagaraMeshRendererProperties]}
  {emitterName=Action_Lines, rendererClasses [&NiagaraMeshRendererProperties]}
  {emitterName=Slash_01, rendererClasses [&NiagaraMeshRendererProperties]}
 ]
 userVariables [
  User.bIsFrontEnd {type=NiagaraBool}
  User.bIsFrontEndPreview {type=NiagaraBool}
  User.SwingDirectionIndex {type=NiagaraInt32}
 ]
}
