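# Niagara ribbon trail for the Keytar Ninja emote, with a single Trail emitter that follows the moving effect.
# Particles spawn by distance traveled as short-lived black ribbon points, facing along an axis rotated by the owner's orientation.
# Curves taper width and quickly fade alpha in then out, multiplied by a time-of-day scalar.
# A local LockFirstParticle module receives the simulation position and owner-rotated vector to pin the ribbon's first particle.
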NiagaraSystem /SparksCosmetics/Emotes/Keytar/Ninja/Effects/NS_Emote_Keytar_Ninja_Trail
dependencies {usedDataInterfaces [&NiagaraDataInterfaceCurve], usedDynamicInputs [], usedModules [&SystemState]}
emitters {Trail {
  emitterUpdateScript [
   &EmitterState()
   &SpawnPerUnit(Module.Use Max Movement Threshold=true, Module.Use Movement Tolerance=false, Spawn Spacing=10, Use Spawn Probability=false)
   /SetVariables(Emitter.HasSpawned:NiagaraBool=$Output.SpawnPerUnit.HasSpawnedThisFrame) {isSetParameters=true}
  ]
  particleSpawnScript [
   &InitializeParticle(
    Color=<rgba 0 0 0 1>
    Lifetime=0.2
    Ribbon Facing Mode=::NewEnumerator1
    Ribbon Facing Vector:Vector3f=<1 0 0>
    Ribbon Width=50
    Ribbon Width Mode=::NewEnumerator1
    UsePositionOffset=false
    Write Material Random=true
   )
   &ApplyInitialForces()
   /SetVariables(Particles.RibbonFacing=&RotateVectorByQuaternion(Vector To Rotate=<0 1 0>, Quaternion=$Engine.Owner.Rotation)) {isSetParameters=true}
  ]
  particleUpdateScript [
   &ParticleState()
   &SolveForcesAndVelocity(Clamp Velocity=false, Limit Acceleration=false)
   &ScaleRibbonWidth(Ribbon Width Scale=&FloatFromCurve(FloatCurve=DataInterface() {propertyValues {Curve {keys [<at 0 1>,<at 1 0.1>]}}}))
   &ScaleColor(
    Scale Alpha=&Multiply_Float(
     A=&FloatFromCurve(FloatCurve=DataInterface() {propertyValues {Curve {keys [
          {interpMode=RCIM_Cubic, tangentMode=RCTM_User, time=0, value=0}
          {interpMode=RCIM_Cubic, tangentMode=RCTM_User, time=0.03, value=1}
          {tangentMode=RCTM_User, time=1, value=0}
         ]}}})
     B=&FNTimeOfDayScalar(Night Scale=1, Morning Scale=1, Evening Scale=1)
    )
    ScaleA=true
    ScaleRGB=true
   )
   &self:LockFirstParticle(
    LockedPosition:NiagaraPosition=/Niagara/DynamicInputs/Helpers/SimulationPosition()
    LockedRotation:Vector3f=&RotateVectorByQuaternion(Vector To Rotate=<0 1 0>, Quaternion=$Engine.Owner.Rotation)
   ) {module=LockFirstParticle}
  ]
  rendererClasses [&NiagaraRibbonRendererProperties]
  renderers [{rendererClass=&NiagaraRibbonRendererProperties}]
 }}
summary {
 systemName=NS_Emote_Keytar_Ninja_Trail
 emitters [{emitterName=Trail, rendererClasses [&NiagaraRibbonRendererProperties]}]
 userVariables [
  User.NormalizedNotifyProgress {type=NiagaraFloat, defaultValue=0, description=""}
  User.WindWhirlSpawnRate {type=NiagaraFloat, defaultValue=1, description=""}
 ]
}
